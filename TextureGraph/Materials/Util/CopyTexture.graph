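# CopyTexture is an unlit utility material in TextureGraph, named for copying a texture.
# It holds a SourceTexture parameter, with an albedo texture as the default, so the input can be swapped per material instance.
# A standalone scalar constant sits beside the sampler, and no links from either node to material outputs appear here.
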Material /TextureGraph/Materials/Util/CopyTexture
graph {
 properties {pixelDepthOffsetMode=PDOM_Legacy, shadingModel=MSM_Unlit, usedShadingModels=Unlit}
 nodes [
  Constant() {r=16}
  TextureSampleParameter2D() {parameterName=SourceTexture, texture=/TextureGraph/Textures/thbnbdaew_1K_Albedo}
 ]
}
