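# Unlit utility material in TextureGraph that, going by its name and the ThumbTex parameter, copies a texture, likely for thumbnails.
# It samples the ThumbTex texture parameter and builds a grayscale version by repeating the red channel through MakeFloat3.
# The SingleChannel scalar blends between the full RGB and that grayscale result, so single-channel textures can display as gray rather than red.
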Material /TextureGraph/Materials/Util/CopyUnlit
graph {
 properties {pixelDepthOffsetMode=PDOM_Legacy, shadingModel=MSM_Unlit, usedShadingModels=Unlit}
 nodes [
  n1 = TextureSampleParameter2D() {parameterName=ThumbTex, texture=/Engine/EngineResources/AICON-Green}
  LinearInterpolate(
   A=@n1.RGB
   B=MaterialFunctionCall(X=@n1.R, Y=@n1.R, Z=@n1.R) {materialFunction=&MakeFloat3}.Result
   Alpha=ScalarParameter() {parameterName=SingleChannel}
  )
 ]
}
