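# Translucent UI material that draws a histogram graph from a HistogramTexture, with switchable red, green, blue and luma channels.
# It samples bins along the texture's first row and divides them by a maximum stored in a second-row pixel, with luma in alpha.
# A Step against vertical UV turns each normalized value into a filled column.
# Show parameters mask and blend the channels, and their clamped sum scales GraphOpacity.
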Material /TextureGraph/Materials/Util/HistogramView
graph {
 properties {blendMode=BLEND_Translucent, materialDomain=MD_UI, pixelDepthOffsetMode=PDOM_Legacy, usedShadingModels=Default Lit}
 nodes [
  Constant3Vector()
  n1 = ScalarParameter() {defaultValue=1, parameterName=ShowR}
  n2 = Constant()
  n3 = Constant() {r=1}
  n4 = TextureCoordinate()
  n5 = Subtract(B=ComponentMask(None=@n4) {g=true})
  n6 = TextureSampleParameter2D(UVs=MaterialFunctionCall(X=@n4, Y=Constant()) {materialFunction=&MakeFloat2}.Result) {parameterName=HistogramTexture, samplerSource=SSM_Clamp_WorldGroupSettings, texture=/TextureGraph/Textures/default/default_Albedo}
  n7 = Constant() {
   desc="The max pixel is the 2nd pixel of second row in histogram trexture\r\nso we calculate the value of u by (1/(width = 256)) * (1.5 = pixel index)"
   r=0.0058
  }
  n8 = TextureSampleParameter2D(UVs=MaterialFunctionCall(X=@n7, Y=Constant() {r=1}) {materialFunction=&MakeFloat2}.Result) {parameterName=HistogramTexture, samplerSource=SSM_Clamp_WorldGroupSettings, texture=&default_Albedo}
  n9 = LinearInterpolate(A=@n2, B=@n3, Alpha=Step(Y=@n5, X=Divide(A=@n6.R, B=@n8.R)))
  n10 = ScalarParameter() {parameterName=ShowG}
  n11 = LinearInterpolate(A=@n2, B=@n3, Alpha=Step(Y=@n5, X=Divide(A=@n6.G, B=@n8.G)))
  n12 = ScalarParameter() {parameterName=ShowB}
  n13 = LinearInterpolate(A=@n2, B=@n3, Alpha=Step(Y=@n5, X=Divide(A=@n6.B, B=@n8.B)))
  n14 = ScalarParameter() {parameterName=ShowLuma}
  n15 = Divide(A=@n6.A, B=@n8.A)
  n16 = Multiply(A=@n14, B=LinearInterpolate(A=@n2, B=@n3, Alpha=Step(Y=@n5, X=@n15)))
  n17 = LinearInterpolate(
   A=MaterialFunctionCall(X=Multiply(A=@n1, B=@n9), Y=Multiply(A=@n10, B=@n11), Z=Multiply(A=@n12, B=@n13)) {materialFunction=&MakeFloat4}.Result
   B=MaterialFunctionCall(X=@n16, Y=@n16, Z=@n16) {materialFunction=&MakeFloat4}.Result
   Alpha=@n14
  )
  Multiply(A=@n17) {constB=0.5}
  Multiply(
   A=Clamp(None=MaterialFunctionCall(f4=@n17) {materialFunction=&AddComponents}.f4)
   B=ScalarParameter() {defaultValue=0.75, parameterName=GraphOpacity}
  )
 ]
}
