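# TextureGraph material function that applies a directional Gaussian blur to an input texture and outputs BlurredTexture.
# A Custom HLSL node, referencing Material Maker's directional blur, samples at one-texel intervals along BlurAngle in degrees and normalizes the weighted RGB sum.
# BlurSteps sets samples per side, while BlurAmount times Sigma sets the Gaussian width.
# UV derivatives are taken before the loop for SampleGrad with a shared bilinear wrap sampler.
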MaterialFunction /TextureGraph/Materials/Util/MF_BlurDirectional
graph {nodes [
  MaterialExpressionFunctionInput_0 = FunctionInput(Preview=TextureObject() {samplerType=SAMPLERTYPE_LinearColor, texture=/TextureGraph/Textures/default/default_Roughness}) {bUsePreviewValueAsDefault=true, inputName=InputTexture, inputType=FunctionInput_Texture2D, sortPriority=-1}
  n1 = Constant2Vector() {g=1, r=1}
  MaterialExpressionMultiply_0 = Multiply(A=TextureProperty(Tex=@MaterialExpressionFunctionInput_0) {property=TMTM_TexelSize}, B=@n1)
  MaterialExpressionMultiply_1 = Multiply(
   A=@n1
   B=FunctionInput(Preview=TextureCoordinate()) {bUsePreviewValueAsDefault=true, inputName=Coordinates, inputType=FunctionInput_Vector2, sortPriority=1}
  )
  MaterialExpressionFunctionInput_6 = FunctionInput(Preview=Constant() {r=6}) {bUsePreviewValueAsDefault=true, inputName=BlurAmount, inputType=FunctionInput_Scalar, sortPriority=3}
  MaterialExpressionFunctionInput_3 = FunctionInput(Preview=Constant()) {bUsePreviewValueAsDefault=true, inputName=BlurAngle, inputType=FunctionInput_Scalar, sortPriority=3}
  MaterialExpressionFunctionInput_4 = FunctionInput(Preview=Constant() {r=2}) {bUsePreviewValueAsDefault=true, inputName=Sigma, inputType=FunctionInput_Scalar, sortPriority=4}
  MaterialExpressionFunctionInput_5 = FunctionInput(Preview=Constant() {r=50}) {bUsePreviewValueAsDefault=true, inputName=BlurSteps, inputType=FunctionInput_Scalar, sortPriority=5}
  n2 = Custom!(
   Tex=@MaterialExpressionFunctionInput_0
   TexelSize=@MaterialExpressionMultiply_0
   UV=@MaterialExpressionMultiply_1
   amount=@MaterialExpressionFunctionInput_6
   angle=@MaterialExpressionFunctionInput_3
   sigma=@MaterialExpressionFunctionInput_4
   steps=@MaterialExpressionFunctionInput_5
  ) {
   code="//Directional Blur\r\n//https://github.com/RodZill4/material-maker/blob/master/addons/material_maker/nodes/directional_blur.mmg\r\n\r\n//return float4(1, 0, 0,1);\r\n//return float4(UV, 0, 1);\r\n//return Tex.SampleGrad(GetMaterialSharedSampler(TexSampler,View.MaterialTextureBilinearWrapedSampler), UV,0,0);\r\n\r\nfloat2 dx = ddx(UV);\r\nfloat2 dy = ddy(UV);\r\n\r\nfloat Radians = radians(angle); \r\nfloat2 e = float2(cos(Radians), -sin(Radians)) * TexelSize;\r\nfloat3 result = float3( 0., 0., 0. );\r\n\r\nfloat weightSum = 0.0;\r\n\r\nfloat sig = sigma *  amount;\r\n\r\nfor( int i = -steps; i < steps; i++ )\r\n{\r\n\tfloat weight = exp( -0.5 * (pow(float(i)/sig, 2.)) ) / ( 6.28318530718 * sig * sig );\r\n\tresult += weight * Tex.SampleGrad(GetMaterialSharedSampler(TexSampler,View.MaterialTextureBilinearWrapedSampler), UV + float(i) * e, dx, dy).xyz;\r\n    weightSum += weight;\r\n}\r\n\r\nreturn result / weightSum;"
   description=Directional Blur
   showCode=true
   inputs [
    {inputName=Tex, input {expression=&self:MaterialExpressionFunctionInput_0}}
    {inputName=TexelSize, input {expression=&self:MaterialExpressionMultiply_0}}
    {inputName=UV, input {expression=&self:MaterialExpressionMultiply_1}}
    {inputName=amount, input {expression=&self:MaterialExpressionFunctionInput_6}}
    {inputName=angle, input {expression=&self:MaterialExpressionFunctionInput_3}}
    {inputName=sigma, input {expression=&self:MaterialExpressionFunctionInput_4}}
    {inputName=steps, input {expression=&self:MaterialExpressionFunctionInput_5}}
   ]
  }
  FunctionOutput(None=@n2) {outputName=BlurredTexture}
 ]}
