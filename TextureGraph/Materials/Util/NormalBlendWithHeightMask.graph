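# Unlit TextureGraph utility material that blends a source texture over a destination texture using a mask, named for normal blending by height.
# The Mask's green channel, scaled by LerpRGB and LerpAlpha, drives separate linear interpolations of color and alpha.
# Only the blended red channel is kept, repeated across RGB and appended with alpha; no normal-specific math appears.
# The Opacity parameter is unused, and no link to a material output is shown.
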Material /TextureGraph/Materials/Util/NormalBlendWithHeightMask
graph {
 properties {pixelDepthOffsetMode=PDOM_Legacy, shadingModel=MSM_Unlit, usedShadingModels=Unlit}
 nodes [
  n1 = TextureSampleParameter2D() {parameterName=DestinationTexture, texture=/TextureGraph/Textures/default/default_Albedo}
  n2 = TextureSampleParameter2D() {parameterName=SourceTexture, texture=&default_Albedo}
  n3 = TextureSampleParameter2D() {parameterName=Mask, samplerType=SAMPLERTYPE_LinearColor, texture=/TextureGraph/Textures/default/default_Displacement}
  n4 = ScalarParameter() {parameterName=LerpRGB}
  n5 = MaterialFunctionCall(Float3=LinearInterpolate(A=@n1.RGB, B=@n2.RGB, Alpha=Multiply(A=@n3.G, B=@n4))) {materialFunction=&BreakOutFloat3Components}
  n6 = ScalarParameter() {parameterName=LerpAlpha}
  MaterialFunctionCall(R=@n5.R, G=@n5.R, B=@n5.R, A=LinearInterpolate(A=@n1.A, B=@n2.A, Alpha=Multiply(A=@n3.G, B=@n6))) {materialFunction=&AppendMany}
  ScalarParameter() {parameterName=Opacity}
 ]
}
