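# RoundedThumb is a masked UI material that renders a texture thumbnail with rounded corners in TextureGraph.
# SingleChannel blends the ThumbTex color toward its red channel replicated as grayscale.
# That color is multiplied by a GeneratedRoundRect mask sized by NormalizedDimensions, then lerped over a tiled checker pattern using texture alpha, so transparent areas show the checker.
# The ShowChecker parameter is declared but feeds nothing.
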Material /TextureGraph/Materials/Util/RoundedThumb
graph {
 properties {blendMode=BLEND_Masked, materialDomain=MD_UI, pixelDepthOffsetMode=PDOM_Legacy, shadingModel=MSM_Unlit, usedShadingModels=Unlit}
 nodes [
  Constant2Vector() {g=0.9, r=0.9}
  n1 = TextureSampleParameter2D() {parameterName=ThumbTex, texture=/TextureGraph/Textures/default/default_Albedo}
  n2 = MaterialFunctionCall(X=@n1.R, Y=@n1.R, Z=@n1.R) {materialFunction=&MakeFloat3}
  n3 = ScalarParameter() {parameterName=SingleChannel}
  n4 = VectorParameter() {parameterName=NormalizedDimensions, defaultValue=<rgba 0.9 0.9 0 0>}
  LinearInterpolate(
   A=MaterialFunctionCall!(UVs=TextureCoordinate(), Tiling=Constant() {r=8}) {materialFunction=/Engine/ArtTools/RenderToTexture/MaterialFunctions/CheckerPattern}.Result
   B=Multiply(
    A=LinearInterpolate(A=@n1.RGB, B=@n2.Result, Alpha=@n3)
    B=MaterialFunctionCall!(UV Coords, Box Dimensions=@n4.RGB, CornerRadius, Sharpness, Center) {materialFunction=/Engine/Functions/Engine_MaterialFunctions01/Gradient/GeneratedRoundRect}.Result
   )
   Alpha=@n1.A
  ) {constAlpha=15}
  ScalarParameter() {defaultValue=1, parameterName=ShowChecker}
 ]
}
